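; AnimNotify for the DeerPlum weapon's jump attack that spawns an effect on the player's weapon when triggered in an animation.
; On notify, it calls the parent, casts the mesh owner to a Fort player pawn and gets that pawn's current weapon mesh.
; It attaches its configured FX System to that mesh's root as a pooled, auto-released effect, then returns false.
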(fn Received_Notify
  (params (MeshComp :type Skeletal%20Mesh%20Component%20Object%20Reference) (Animation :type Anim%20Sequence%20Base%20Object%20Reference) (EventReference :type Anim%20Notify%20Event%20Reference%20Structure%20%28by%20ref%29))
  (|CallParentFunction MeshComp Animation EventReference)
  (bind _asfort_player_pawn (Utilities|Casting|CastToFortPlayerPawn (Components|GetOwner MeshComp)))
  (Effects|SpawnFXSystemAttached :FXSystemAsset (Variables|Default|GetFX%20System) :AttachToComponent (Variables|/Script/FortniteGame.FortWeapon|GetWeaponMesh (Weapon|GetCurrentWeapon _asfort_player_pawn)) :AttachPointName "root" :Scale "1.000000,1.000000,1.000000" :LocationType "KeepRelativeOffset" :bAutoDestroy true :PoolingMethod "AutoRelease" :bAutoActivate true)
  (return false))
