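# Translucent, unlit, two-sided mesh-particle material for a DeerPlum smash effect.
# Particle color, optionally multiplied by desaturation-adjusted panning noise, feeds the emissive adjustment function.
# Opacity combines particle alpha with a packed flare texture's blue channel, U-tiled by dynamic TileX, plus optional depth fade disabled on mobile.
# Both contrast lerps use Contrast_Neg at each end, cancelling red/green channel selection; a cloud-noise branch is unused.
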Material /WildEstateWeapons/Assets/FX/DeerPlum/M_DeerPlum_CS_Smash
graph {
 properties {
  bEnableResponsiveAA=true
  bUsedWithMeshParticles=true
  bUsedWithNiagaraMeshParticles=true
  blendMode=BLEND_Translucent
  pixelDepthOffsetMode=PDOM_Legacy
  shadingModel=MSM_Unlit
  twoSided=true
  usedShadingModels=Unlit
 }
 nodes [
  n1 = TextureCoordinate()
  n2 = AppendVector(
   A=Multiply(A=ComponentMask(None=@n1) {r=true}, B=ScalarParameter() {defaultValue=1, group=ChromaticNoise, parameterName=Noise-TilingX})
   B=Multiply(A=ComponentMask(None=@n1) {g=true}, B=ScalarParameter() {group=ChromaticNoise, parameterName=Noise-TilingY})
  )
  n3 = AppendVector(
   A=ScalarParameter() {group=ChromaticNoise, parameterName=NoisePanning-X}
   B=ScalarParameter() {defaultValue=0.2, group=ChromaticNoise, parameterName=NoisePanning-Y}
  )
  n4 = Desaturation(
   None=TextureSampleParameter2D(UVs=Panner(Coordinate=@n2, Speed=@n3)) {group=ChromaticNoise, parameterName=NoiseTexture, texture=/Game/Global/Textures/Default/Noise/T_Noise_Cloud}.RGB
   Fraction=ScalarParameter() {defaultValue=-0.5, group=ChromaticNoise, parameterName=Noise-Desat_Frac}
  )
  n5 = ParticleColor()
  MaterialFunctionCall(Emissive=StaticSwitchParameter(True=Multiply(A=@n5.RGB, B=@n4), False=@n5.RGB) {group=ChromaticNoise, parameterName=UseChromaticNoise}) {materialFunction=&MF_FX_EmissiveAdjustments}
  n6 = Multiply(A=TextureCoordinate(), B=Constant2Vector() {g=0.2, r=2})
  Multiply(A=Power(Base=TextureSample(UVs=@n6) {texture=&AtmosphericCloudNoise01}.RGB) {constExponent=3}) {constB=5}
  n7 = TextureCoordinate()
  n8 = DynamicParameter() {defaultValue=<rgba 0 0 0 1>, paramNames [Contrast_Neg, Contrast_Pos, WhichChannel?, TileX]}
  n9 = AppendVector(A=Multiply(A=ComponentMask(None=@n7) {r=true}, B=Ceil(None=@n8.TileX)), B=ComponentMask(None=TextureCoordinate()) {g=true})
  Multiply(A=@n9, B=Constant2Vector() {g=-1, r=1})
  n10 = TextureSample(UVs=@n9) {samplerType=SAMPLERTYPE_LinearColor, texture=/Game/Effects/Fort_Effects/Textures/Flares/T_Smash_Flare_Packed}
  n11 = LinearInterpolate(A=@n8.Contrast_Neg, B=@n8.Contrast_Neg, Alpha=@n10.R)
  n12 = LinearInterpolate(A=@n8.Contrast_Neg, B=@n8.Contrast_Neg, Alpha=@n10.G)
  n13 = Multiply(A=@n5.A, B=Multiply(A=LinearInterpolate(A=@n11, B=@n12, Alpha=@n8.WhichChannel?), B=@n10.B))
  n14 = DepthFade(Opacity=@n13, FadeDistance=ScalarParameter() {defaultValue=120, parameterName=FadeDistance})
  StaticSwitchParameter(True=FeatureLevelSwitch(Default=@n14, ES2=@n13, ES3_1=@n13, SM4_REMOVED=@n14), False=@n13) {parameterName=UseDepthFade}
 ]
}
