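# Unlit, two-sided translucent material for Niagara mesh particles, apparently the animated toon-line effect of the DeerPlum weapon FX.
# A grayscale saw-lines texture is tiled and panned, then shaped by parameterized power, multiply and add, passed through CheapContrast and multiplied by particle alpha.
# Tiling and speed reuse their X channel for both axes, leaving the Y channels and the contrast channel unconnected.
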Material /WildEstateWeapons/Assets/FX/DeerPlum/M_Deerplum_toonlines
graph {
 properties {
  bAutomaticallySetUsageInEditor=false
  bUsedWithNiagaraMeshParticles=true
  blendMode=BLEND_Translucent
  previewMesh=&/SM_DeerPlum_Lines
  shadingModel=MSM_Unlit
  twoSided=true
  usedShadingModels=Unlit
 }
 nodes [
  n1 = VectorParameter() {parameterName=Texture Option, channelNames {a=Speed Y, b=Speed X, g=tile Y, r=tile X}, defaultValue=<rgba 0.35 3 0 -1>}
  n2 = Panner(Coordinate=Multiply(A=TextureCoordinate(), B=AppendVector(A=@n1.tile X, B=@n1.tile X)), Speed=AppendVector(A=@n1.Speed X, B=@n1.Speed X)) {bFractionalPart=true}
  n3 = VectorParameter() {parameterName=Texture Edit Options, channelNames {a=contrast, b=add, g=mul, r=pow}, defaultValue=<rgba 1 1 -0.3444 0>}
  n4 = Multiply(
   A=Power(
    Base=TextureSample(UVs=@n2) {samplerType=SAMPLERTYPE_LinearGrayscale, texture=&/T_SawLines_basecolor}.R
    Exp=@n3.pow
   )
   B=@n3.mul
  )
  Multiply(
   A=ParticleColor().A
   B=MaterialFunctionCall(In=Add(A=@n4, B=@n3.add)) {materialFunction=&CheapContrast}.Result
  )
 ]
}
