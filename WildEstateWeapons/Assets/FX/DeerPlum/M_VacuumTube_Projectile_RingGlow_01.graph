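# Translucent unlit Niagara particle material for the ring glow on the VacuumTube projectile effect, with emissive driven by particle color.
# A smoke disk texture pans in radial coordinates, optionally rotated and noise-distorted via UseDistortion, then is smoothstepped and cut into a ring by two sphere masks.
# A radial noise mask offset by the Erode dynamic parameter dissolves the ring.
# Depth fade and view-angle fade soften the resulting opacity.
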Material /WildEstateWeapons/Assets/FX/DeerPlum/M_VacuumTube_Projectile_RingGlow_01
graph {
 properties {
  bUsedWithNiagaraMeshParticles=true
  bUsedWithNiagaraSprites=true
  blendMode=BLEND_Translucent
  pixelDepthOffsetMode=PDOM_Legacy
  shadingModel=MSM_Unlit
  usedShadingModels=Unlit
 }
 nodes [
  DynamicParameter() {parameterIndex=2, defaultValue=<rgba -1 0 0 1>, paramNames [na,na,na,na]}
  DynamicParameter() {parameterIndex=3, defaultValue=<rgba -1 0 0 1>, paramNames [na,na,na,na]}
  MaterialFunctionCall() {materialFunction=&DebugTimeSine}
  MaterialFunctionCall(Emissive=ParticleColor().RGB) {materialFunction=&MF_FX_EmissiveAdjustments}
  n1 = TextureCoordinate()
  n2 = Constant2Vector() {g=-0.5, r=-0.5}
  n3 = Multiply(
   A=MaterialFunctionCall(Vector or UVs=Add(A=@n1, B=@n2)) {materialFunction=&VectorToRadialValue}.Radial Coordinates
   B=Constant2Vector() {g=3, r=2.5}
  ) {constB=2}
  n4 = Constant() {r=1}
  n5 = Add(
   A=LinearInterpolate(Alpha=DynamicParameter() {parameterIndex=1, defaultValue=<rgba 0 0 0 1>, paramNames [Erode,na,na,na]}.Erode) {constA=1, constB=-1}
   B=TextureSample(UVs=MaterialFunctionCall(UVs=@n3, Texture Scale=@n4) {materialFunction=&ScaleUVsByCenter}.UVs) {texture=/Game/Effects/Fort_Effects/Textures/T_Smooth_tile}.R
  )
  n6 = Constant() {r=1}
  n7 = Multiply(A=Saturate(None=MaterialFunctionCall(In=@n5, Contrast=@n6) {materialFunction=&CheapContrast}.Result))
  n8 = Add(A=TextureCoordinate(), B=Constant2Vector() {g=-0.5, r=-0.5})
  n9 = Constant2Vector()
  n10 = Multiply(A=DynamicParameter() {defaultValue=<rgba -1 0 0 1>, paramNames [ExplosionPan, DistortAmount, RotateAmount, DistortPanSpeed]}.RotateAmount) {constB=-1}
  n11 = TextureCoordinate()
  n12 = Multiply(
   A=Constant2Vector() {g=1, r=1}
   B=DynamicParameter() {defaultValue=<rgba -1 0 0 1>, paramNames [ExplosionPan, DistortAmount, RotateAmount, DistortPanSpeed]}.DistortPanSpeed
  ) {constB=0.25}
  n13 = TextureSampleParameter2D(UVs=MaterialFunctionCall(UVs=Panner(Coordinate=@n11, Speed=@n12) {speedX=-1, speedY=1}, Texture Scale=Constant() {r=0.5}) {materialFunction=&ScaleUVsByCenter}.UVs) {parameterName=DistortTexture2, texture=&T_Smooth_tile}
  n14 = DynamicParameter() {defaultValue=<rgba -1 0 0 1>, paramNames [ExplosionPan, DistortAmount, RotateAmount, DistortPanSpeed]}
  n15 = Add(
   A=LinearInterpolate(
    A=@n8
    B=MaterialFunctionCall(UVs=@n8, Rotation Center=@n9, "Rotation Angle (0-1)"=@n10) {materialFunction=&CustomRotator}.Rotated Values
    Alpha=ScalarParameter() {parameterName=UseDistortion}
   ) {constB=0}
   B=LinearInterpolate(B=Multiply(A=@n13.G, B=@n14.DistortAmount) {constB=0.25}, Alpha=ScalarParameter() {parameterName=UseDistortion}) {constB=0}
  )
  n16 = TextureSample(UVs=Panner(
    Coordinate=MaterialFunctionCall(Vector or UVs=@n15) {materialFunction=&VectorToRadialValue}.Radial Coordinates
    Time=DynamicParameter() {defaultValue=<rgba -1 0 0 1>, paramNames [ExplosionPan, DistortAmount, RotateAmount, DistortPanSpeed]}.ExplosionPan
   ) {speedY=-0.25}) {texture=/Game/Effects/Fort_Effects/Textures/Smoke/T_Smoke_Bomb_Disk2}
  n17 = LinearInterpolate(
   A=@n16.G
   B=Multiply(
    A=DynamicParameter() {defaultValue=<rgba -1 0 0 1>, paramNames [ExplosionPan, DistortAmount, RotateAmount, DistortPanSpeed]}.DistortAmount
    B=@n16.G
   )
   Alpha=ScalarParameter() {parameterName=UseDistortion}
  )
  n18 = ScalarParameter() {defaultValue=0.1, parameterName=Min}
  n19 = ScalarParameter() {defaultValue=1, parameterName=Max}
  n20 = SphereMask(A=TextureCoordinate(), B=Constant2Vector() {g=0.5, r=0.5}, Radius=Constant() {r=0.2}, Hardness=Constant() {r=0.4})
  n21 = Multiply(
   A=Multiply(
    A=MaterialFunctionCall(Alpha=@n17, Min=@n18, Max=@n19) {materialFunction=&SmoothStep}.Result
    B=OneMinus(None=@n20)
   )
   B=SphereMask(A=TextureCoordinate(), B=Constant2Vector() {g=0.5, r=0.5}, Radius=Constant() {r=0.4}, Hardness=Constant() {r=-1})
  )
  Multiply(
   A=DepthFade(Opacity=Multiply(A=@n7, B=@n21), FadeDistance=ScalarParameter() {defaultValue=50, parameterName=DepthFade})
   B=MaterialFunctionCall("FadeSensitivity (optional)"=ScalarParameter() {parameterName="FadeSensitivity (optional) (S)"}) {materialFunction=&FadeBasedOnViewAngle}.Result
  )
 ]
}
