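# Niagara system for DeerPlum that produces camera lens dirt as small sprites from one GPU emitter, DirtOnScreen.
# It continuously spawns randomly sized specks in a thin, wide box that forms a screen-like plane, with slight downward gravity.
# A spherical-falloff RegionMask drives a size lerp so specks vary by position, while a curve pops each in and shrinks it out.
# Torus placement, mask-based full shrink and color fade modules are disabled.
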NiagaraSystem /WildEstateWeapons/Assets/FX/DeerPlum/NS_DeerPlum_CS_Camera_Lens_Dirt
dependencies {usedDataInterfaces [&NiagaraDataInterfaceCurve], usedDynamicInputs [], usedModules [&SystemState]}
emitters {DirtOnScreen {
  simTarget=GPUComputeSim
  emitterUpdateScript [
   &EmitterState()
   &SpawnRate(SpawnRate=200, Use Spawn Probability=true)
  ]
  particleSpawnScript [
   &InitializeParticle(
    Color=<rgba 1 1 1 50>
    Lifetime Mode=::NewEnumerator1
    Sprite Rotation Mode=::NewEnumerator2
    Sprite Size Mode=::NewEnumerator3
    Sprite UV Mode=::NewEnumerator3
    Uniform Sprite Size=&Multiply_Float(A=&RandomRangeFloat(Minimum=2, Maximum=5), B=0.65)
    UsePositionOffset=false
    Write Material Random=true
   )
   &TorusLocation(Align Torus With Axis=true, Handle Radius=60, MaskSpawnGroup=false, Module.Override Randomness=false, Torus Axis=<1 0 0>) {enabled=false}
   &ShapeLocation(
    Box Midpoint=<0.5 0.5 0.9>
    Box Size=<10 240 120>
    MaskSpawnGroup=false
    Override Seed=false
    Shape Primitive=::NewEnumerator2
    Surface Only=false
    UseOwnerScale=true
   )
   &RegionMask(Region Falloff=0.3, Region Radius=75, UseOffset=true, UseRotation=false)
  ]
  particleUpdateScript [
   &ParticleState()
   &ScaleSpriteSize(Uniform Curve Sprite Scale=DataInterface() {propertyValues {Curve {defaultValue=0, keys [<at 0 0>, <at 0.025 1>, {tangentMode=RCTM_User, time=0.85, value=1}, {tangentMode=RCTM_User, time=1, value=0}]}}}) {module=ScaleSpriteSize002}
   &ScaleSpriteSize(
    Scale Sprite Size Mode=::NewEnumerator0
    Uniform Scale Factor=&Lerp_Float(A=1.4, B=0.3, Alpha=$StackContext.RegionMask.Float)
   ) {module=ScaleSpriteSize001}
   &ScaleSpriteSize(
    Scale Sprite Size Mode=::NewEnumerator0
    Uniform Scale Factor=&Lerp_Float(A=1, B=0, Alpha=$StackContext.RegionMask.Float)
   ) {enabled=false}
   &ScaleColor(
    Scale Alpha=&FloatFromCurve(
     FloatCurve=DataInterface() {propertyValues {Curve {keys [
         {tangentMode=RCTM_User, time=0, value=1}
         {arriveTangent=-0.0405, interpMode=RCIM_Cubic, leaveTangent=-0.0405, tangentMode=RCTM_User, time=0.436, value=0.6364}
         {arriveTangent=-0.5488, interpMode=RCIM_Cubic, leaveTangent=-0.5488, tangentMode=RCTM_User, time=1, value=0}
        ]}}}
     Scale Curve=50
    )
    ScaleA=true
    ScaleRGB=true
   ) {enabled=false}
   &GravityForce(Gravity=<0 0 -0.5>)
   &SolveForcesAndVelocity(Clamp Velocity=false, Limit Acceleration=false)
  ]
  rendererClasses [&NiagaraSpriteRendererProperties]
  renderers [{rendererClass=&NiagaraSpriteRendererProperties}]
 }}
summary {
 systemName=NS_DeerPlum_CS_Camera_Lens_Dirt
 emitters [{emitterName=DirtOnScreen, simTarget=GPUComputeSim, rendererClasses [&NiagaraSpriteRendererProperties]}]
 userVariables []
}
